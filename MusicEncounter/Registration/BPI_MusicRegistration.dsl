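; BPI_MusicRegistration is a Blueprint interface for music registration in the MusicEncounter system.
; It declares GetMusicEventTag, which takes an Audio Component reference and returns an event tag, and GetMusicGroupName, which returns the name of a music group.
; Both functions carry only empty default returns, leaving the behavior to implementers.
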(fn GetMusicEventTag
  (params (Audio%20Component :type Audio%20Component%20Object%20Reference))
  (return 0))

(fn GetMusicGroupName
  (return "None"))
