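; BP_MusicRegistrar declares the functions for registering music sources and querying which has the highest priority.
; Register, unregister, update and lookup calls all take an Audio Component to identify the entry.
; Registration also takes a Music Registration Interface, while updates take a Music Registration Data structure.
; The functions shown contain no logic: the setters are empty and both getters return zero defaults.
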(fn GetHighestPriorityMusicData
  (return 0))

(fn GetRegisteredMusicData
  (params (Audio%20Component :type Audio%20Component%20Object%20Reference))
  (return 0))

(fn RegisterMusicbyComponent
  (params (Music%20Registration%20Interface :type Music%20Registration%20Interface) (Audio%20Component :type Audio%20Component%20Object%20Reference)))

(fn UnregisterMusic
  (params (Audio%20Component :type Audio%20Component%20Object%20Reference)))

(fn UpdateRegistration
  (params (Audio%20Component :type Audio%20Component%20Object%20Reference) (Registration%20Data :type Music%20Registration%20Data%20Structure)))
