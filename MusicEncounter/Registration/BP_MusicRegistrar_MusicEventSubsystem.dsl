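; Music registrar that bridges audio components to the MusicEventSubsystem for the music encounter system.
; It registers components with an event tag and group name from BPI_MusicRegistration, and removes them on unregister.
; Updates change the event's category when the registration tag is valid, then change its group.
; Queries wrap a component's instance, or the highest-priority one, as RegisteredMusicData holding attenuation state and priority.
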(fn GetHighestPriorityMusicData
  (bind _self self)
  (bind _returnvalue (Utilities|GetOuterObject _self))
  (bind _returnvalue_2 (GameInstanceSubsystems|GetMusicEventSubsystem _returnvalue))
  (bind _returnvalue_3 (Audio|GetHighestPriorityInstance _returnvalue_2))
  (Utilities|IsValid _returnvalue
    (:Is%20Valid
      (Utilities|IsValid _returnvalue_2
        (:Is%20Valid
          (Utilities|IsValid _returnvalue_3
            (:Is%20Valid
              (return (Utilities|Struct|MakeRegisteredMusicData (RuntimeData|IsWithinAttenuation _returnvalue_3) (RuntimeData|GetPriority _returnvalue_3))))))))))

(fn GetRegisteredMusicData
  (params (Audio%20Component :type Audio%20Component%20Object%20Reference))
  (bind _audio_component (Variables|GetAudio%20Component))
  (bind _returnvalue (GameInstanceSubsystems|GetMusicEventSubsystem Audio%20Component))
  (bind _returnvalue_2 (Audio|GetEventInstance _returnvalue _audio_component))
  (Utilities|IsValid Audio%20Component
    (:Is%20Valid
      (Utilities|IsValid _returnvalue
        (:Is%20Valid
          (Utilities|IsValid _returnvalue_2
            (:Is%20Valid
              (return (Utilities|Struct|MakeRegisteredMusicData (RuntimeData|IsWithinAttenuation _returnvalue_2) (RuntimeData|GetPriority _returnvalue_2))))))))))

(fn RegisterMusicbyComponent
  (params (Music%20Registration%20Interface :type Music%20Registration%20Interface) (Audio%20Component :type Audio%20Component%20Object%20Reference))
  (bind _music_event_tag (MusicRegistration|MusicEventSubsystem|GetMusicEventTag Music%20Registration%20Interface Audio%20Component :!class /MusicEncounter/Registration/BPI_MusicRegistration.BPI_MusicRegistration_C))
  (bind _music_group_name (MusicRegistration|MusicEventSubsystem|GetMusicGroupName (Variables|GetMusic%20Registration%20Interface) :!class /MusicEncounter/Registration/BPI_MusicRegistration.BPI_MusicRegistration_C))
  (bind _audio_component (Variables|GetAudio%20Component)
    (:then
      (Audio|AddEventbyComponent (GameInstanceSubsystems|GetMusicEventSubsystem _audio_component) _audio_component _music_event_tag "(GameplayTags=)" 0 _music_group_name))
    (:else)))

(fn UnregisterMusic
  (params (Audio%20Component :type Audio%20Component%20Object%20Reference))
  (bind _returnvalue (GameInstanceSubsystems|GetMusicEventSubsystem Audio%20Component))
  (Utilities|IsValid Audio%20Component
    (:Is%20Valid
      (Utilities|IsValid _returnvalue
        (:Is%20Valid
          (Audio|RemoveEvent _returnvalue (Variables|GetAudio%20Component)))))))

(fn UpdateRegistration
  (params (Audio%20Component :type Audio%20Component%20Object%20Reference) (Registration%20Data :type Music%20Registration%20Data%20Structure))
  (bind _registration_data (Variables|GetRegistration%20Data))
  (bind _registrationtag_2_a319b0c94bfc8ecc50badc819c8b314a (Utilities|Struct|BreakMusicRegistrationData _registration_data))
  (Utilities|IsValid Audio%20Component
    (:Is%20Valid
      (Utilities|FlowControl|Sequence
        (:then_0
          (if (GameplayTags|IsGameplayTagValid _registrationtag_2_a319b0c94bfc8ecc50badc819c8b314a)
            (Audio|ChangeEventCategory (GameInstanceSubsystems|GetMusicEventSubsystem (Variables|GetAudio%20Component)) (Variables|GetAudio%20Component) _registrationtag_2_a319b0c94bfc8ecc50badc819c8b314a)))
        (:then_1
          (Utilities|FlowControl|Sequence
            (:then_0
              (Audio|ChangeEventGroup (GameInstanceSubsystems|GetMusicEventSubsystem (Variables|GetAudio%20Component)) (Variables|GetAudio%20Component) (Utilities|Struct|BreakMusicRegistrationData (Variables|GetRegistration%20Data))))))))))
